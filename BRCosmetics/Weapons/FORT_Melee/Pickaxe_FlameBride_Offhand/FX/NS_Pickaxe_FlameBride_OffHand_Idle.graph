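# Niagara idle effect for the FlameBride offhand pickaxe that sheds tumbling red rose meshes with flame sprites.
# Rose spawns meshes around the sampled skeletal mesh per distance moved, with spacing gated by owner speed.
# Fire reads Rose particles to spawn SubUV flames on them, tinted by color curves and time of day.
# Front-end variants RoseFE and FireFE, driven by bIsFrontEndPreview, are present but disabled.
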NiagaraSystem /BRCosmetics/Weapons/FORT_Melee/Pickaxe_FlameBride_Offhand/FX/NS_Pickaxe_FlameBride_OffHand_Idle
dependencies {
 usedDataInterfaces [
  &NiagaraDataInterfaceParticleRead
  &NiagaraDataInterfaceCurve
  &NiagaraDataInterfaceSpriteRendererInfo
  &NiagaraDataInterfaceColorCurve
 ]
 usedDynamicInputs [&ClampFloat]
 usedModules [
  &self:SystemSpawnScript.SystemScriptSource.SystemScriptGraph.NiagaraNodeAssignment_0.SetVariables_#1
  &SystemState
  &SampleSkeletalMesh
  &InterpolateOverTime
  &self:SystemSpawnScript.SystemScriptSource.SystemScriptGraph.NiagaraNodeAssignment_1.SetVariables_#2
 ]
}
emitters {
 Fire {
  emitterUpdateScript [
   &EmitterState(Inactive Response=::NewEnumerator1, Life Cycle Mode=::NewEnumerator1, Loop Delay=0.03, Loop Duration=1, UseLoopDelay=true)
   &SpawnParticlesFromOtherEmitter(
    Attribute Reader=DataInterface() {propertyValues {EmitterBinding {bindingMode=Other, emitterName=Rose}}}
    Use Maximum Age=false
    Use Minimum Age=false
   )
  ]
  particleSpawnScript [
   &InitializeParticle(Lifetime=0.3, Sprite Rotation Mode=::NewEnumerator2, Sprite Size Mode=::NewEnumerator1, UsePositionOffset=false, Write Material Random=true)
   &SampleParticlesFromOtherEmitter(Velocity Sampling=::NewEnumerator0)
  ]
  particleUpdateScript [
   &ParticleState()
   &ScaleSpriteSize(
    Uniform Curve Scale=&MakeCustomFloatFromBool(Bool=$User.bIsFrontEndPreview, True Float=0.6, False Float=0.9)
    Uniform Curve Sprite Scale=DataInterface() {propertyValues {Curve {keys [
        {arriveTangent=0.4113, interpMode=RCIM_Cubic, leaveTangent=0.4113, tangentMode=RCTM_User, time=0, value=0.5}
        {arriveTangent=0.5, interpMode=RCIM_Cubic, leaveTangent=0.5, time=0.15, value=1.4}
        {arriveTangent=-0.2177, interpMode=RCIM_Cubic, leaveTangent=-0.2177, tangentMode=RCTM_User, time=1, value=1}
       ]}}}
   )
   &GravityForce(Gravity=<0 0 -100>) {enabled=false}
   &Drag(Drag=100, Use Linear Drag=true, Use Rotational Drag=true)
   &InheritSourceMovement(
    Applied Position Falloff Curve=DataInterface() {propertyValues {Curve {keys [<at 0 1>,<at 1 0.4>]}}}
    Applied Position Falloff Distance=100
    Applied Position Scale=<0.7 0.7 0.8>
    ApplyForce=false
    ApplyPosition=true
    ApplyPositionFalloff=true
    ApplyPositionFalloffCurve=true
   )
   &SolveForcesAndVelocity(Clamp Velocity=false, Limit Acceleration=false)
   &SubUVAnimation(
    End Frame Range Override=15
    "Sprite Renderer "=DataInterface() {propertyValues {SpriteRenderer=&self:Fire_0.NiagaraSpriteRendererProperties_2}}
    Start Frame Range Override=5
    UseEndFrame=true
    UseStartFrame=true
   )
   &ScaleColor(
    Scale Alpha=&FloatFromCurve(
     FloatCurve=DataInterface() {propertyValues {Curve {keys [
         <at 0 0 RCIM_Cubic>
         {arriveTangent=0.0413, interpMode=RCIM_Cubic, leaveTangent=0.0413, tangentMode=RCTM_User, time=0.1, value=1}
         {arriveTangent=-1.15, interpMode=RCIM_Cubic, leaveTangent=-1.15, tangentMode=RCTM_User, time=0.4, value=0.3}
         <at 0.8 0 RCIM_Cubic>
        ]}}}
     Scale Curve=0.3
    )
    Scale RGB=&MakeVectorFromLinearColorRGB(Color=&ColorFromCurve(ColorCurve=DataInterface() {propertyValues {
        AlphaCurve {keys []}
        BlueCurve {keys [<at 0.0717 0.0162>, {arriveTangent=-0.035, interpMode=RCIM_Cubic, leaveTangent=-0.035, time=0.2309, value=0}, <at 0.536 0>]}
        GreenCurve {keys [<at 0.0717 0.2>, {arriveTangent=-0.4201, interpMode=RCIM_Cubic, leaveTangent=-0.4201, time=0.2309, value=0.12}, <at 0.536 0.005>]}
        RedCurve {keys [<at 0.0717 1>, {arriveTangent=-2.0644, interpMode=RCIM_Cubic, leaveTangent=-2.0644, time=0.2309, value=1}, <at 0.536 0.0417>]}
       }}))
    ScaleA=true
    ScaleRGB=true
   )
   &ScaleColor(
    Scale Alpha=&FNTimeOfDayScalar(Night Scale=0.6, Morning Scale=0.7, Evening Scale=0.7)
    Scale RGB=&MakeVectorFromLinearColorRGB(Color=&FNTimeOfDayColor(Night Color=<rgba 1 0.9948 0.2694 1>, Morning Color=<rgba 1 1 1 1>, Day Color=<rgba 1 1 1 1>, Evening Color=<rgba 1 1 1 1>))
    ScaleA=true
    ScaleRGB=true
   ) {module=ScaleColor001}
  ]
  rendererClasses [&NiagaraSpriteRendererProperties]
  renderers [{rendererClass=&NiagaraSpriteRendererProperties}]
 }
 FireFE {
  bEnabled=false
  emitterUpdateScript [
   &EmitterState(Inactive Response=::NewEnumerator1, Life Cycle Mode=::NewEnumerator1, Loop Duration=1.2, UseLoopDelay=true)
   &SpawnParticlesFromOtherEmitter(
    Attribute Reader=DataInterface() {propertyValues {EmitterBinding {bindingMode=Other, emitterName=RoseFE}}}
    Spawn Rate=&MakeCustomFloatFromBool(
     Bool=$User.bIsFrontEndPreview
     True Float=&FloatFromCurve(
      FloatCurve=DataInterface() {propertyValues {Curve {keys [<at 0.4 0>,<at 0.5 1>,<at 0.6 1>,<at 0.7 0>,<at 1.2 0>]}}}
      CurveIndex=$Emitter.NormalizedLoopAge
      Scale Curve=10
     )
     False Float=0
    )
    Use Maximum Age=false
    Use Minimum Age=false
   )
  ]
  particleSpawnScript [
   &InitializeParticle(Lifetime=0.6, Sprite Rotation Mode=::NewEnumerator2, Sprite Size Mode=::NewEnumerator1, UsePositionOffset=false, Write Material Random=true)
   &SampleParticlesFromOtherEmitter(Velocity Sampling=::NewEnumerator0)
  ]
  particleUpdateScript [
   &ParticleState()
   &ScaleSpriteSize(Uniform Curve Sprite Scale=DataInterface() {propertyValues {Curve {keys [
        {arriveTangent=0.4113, interpMode=RCIM_Cubic, leaveTangent=0.4113, tangentMode=RCTM_User, time=0, value=0.5}
        {arriveTangent=0.3, interpMode=RCIM_Cubic, leaveTangent=0.3, time=0.1366, value=1.4043}
        {arriveTangent=-0.2177, interpMode=RCIM_Cubic, leaveTangent=-0.2177, tangentMode=RCTM_User, time=1, value=0.8}
       ]}}})
   &GravityForce(Gravity=<0 0 -100>) {enabled=false}
   &Drag(Drag=100, Use Linear Drag=true, Use Rotational Drag=true)
   &InheritSourceMovement(
    Applied Position Falloff Curve=DataInterface() {propertyValues {Curve {keys [<at 0 1>,<at 1 0.4>]}}}
    Applied Position Falloff Distance=100
    ApplyForce=false
    ApplyPosition=true
    ApplyPositionFalloff=true
    ApplyPositionFalloffCurve=false
   ) {enabled=false}
   &SolveForcesAndVelocity(Clamp Velocity=false, Limit Acceleration=false)
   &SubUVAnimation(
    End Frame Range Override=12
    "Sprite Renderer "=DataInterface() {propertyValues {SpriteRenderer=&self:FireFE_0.NiagaraSpriteRendererProperties_2}}
    Start Frame Range Override=6
    UseEndFrame=true
    UseStartFrame=true
   )
   &ScaleColor(
    Scale Alpha=&FloatFromCurve(
     FloatCurve=DataInterface() {propertyValues {Curve {keys [
         <at 0 0 RCIM_Cubic>
         {arriveTangent=0.0413, interpMode=RCIM_Cubic, leaveTangent=0.0413, tangentMode=RCTM_User, time=0.2, value=1}
         {arriveTangent=-1.25, interpMode=RCIM_Cubic, leaveTangent=-1.25, time=0.5, value=0.25}
         <at 1 0 RCIM_Cubic>
        ]}}}
     Scale Curve=0.15
    )
    Scale RGB=&MakeVectorFromLinearColorRGB(Color=&ColorFromCurve(ColorCurve=DataInterface() {propertyValues {
        AlphaCurve {keys []}
        BlueCurve {keys [<at 0.0717 0.0162>, {arriveTangent=-0.035, interpMode=RCIM_Cubic, leaveTangent=-0.035, time=0.2309, value=0.02}, <at 0.536 0>]}
        GreenCurve {keys [<at 0.0717 0.2>, {arriveTangent=-0.4201, interpMode=RCIM_Cubic, leaveTangent=-0.4201, time=0.2309, value=0.167}, <at 0.536 0.005>]}
        RedCurve {keys [<at 0.0717 1>, {arriveTangent=-2.0644, interpMode=RCIM_Cubic, leaveTangent=-2.0644, time=0.2309, value=1}, <at 0.536 0.0417>]}
       }}))
    ScaleA=true
    ScaleRGB=true
   )
   &ScaleColor(
    Scale Alpha=&FNTimeOfDayScalar(Night Scale=0.6, Morning Scale=0.7, Evening Scale=0.7)
    Scale RGB=&MakeVectorFromLinearColorRGB(Color=&FNTimeOfDayColor(Night Color=<rgba 1 0.9948 0.2694 1>, Morning Color=<rgba 1 1 1 1>, Day Color=<rgba 1 1 1 1>, Evening Color=<rgba 1 1 1 1>))
    ScaleA=true
    ScaleRGB=true
   ) {module=ScaleColor001}
  ]
  rendererClasses [&NiagaraSpriteRendererProperties]
  renderers [{rendererClass=&NiagaraSpriteRendererProperties}]
 }
 Rose {
  emitterUpdateScript [
   &EmitterState(Inactive Response=::NewEnumerator1, Life Cycle Mode=::NewEnumerator1, Loop Duration=1, UseLoopDelay=false)
   &SpawnPerUnit(
    Max Movement Threshold=600
    Module.Use Max Movement Threshold=true
    Module.Use Movement Tolerance=false
    Spawn Probability=1
    Spawn Spacing:NiagaraFloat=&Multiply_Float(
     A=25
     B=&MakeCustomFloatFromBool(
      Bool=&SetBoolByFloatComparison(A=&VectorLength(Vector=$Engine.Owner.Velocity), B=300)
      True Float=1
      False Float=0
     )
    )
    Use Spawn Probability=true
   )
   &SpawnRate(
    SpawnRate=&MakeCustomFloatFromBool(Bool=$User.SwingSpawn, True Float=0, False Float=50)
    Use Spawn Probability=false
   ) {enabled=false}
   &SpawnBurst_Instantaneous(
    Spawn Count=&MakeCustomIntFromBool(Bool=$User.bIsFrontEndPreview, True INT=8, False INT=6)
    Use Spawn Probability=false
    UseLoopCountLimit=false
   ) {enabled=false}
  ]
  particleSpawnScript [
   &InitializeParticle(
    Lifetime=0.2
    Mesh Scale Max=<1 1 1>
    Mesh Scale Min=<0.5 0.5 0.5>
    Mesh Scale Mode=::NewEnumerator2
    Sprite Size Mode=::NewEnumerator1
    UsePositionOffset=false
    Write Material Random=true
   )
   &InitialMeshOrientation(
    Mesh Orientation Mode=::NewEnumerator2
    Rotation=&RandomRangeVector(Minimum=<0 0 0>, Maximum=<1 1 1>)
    Use Rotation Vector=true
   )
   &ShapeLocation(MaskSpawnGroup=false, Offset Mode=::NewEnumerator0, Override Seed=false, Sphere Radius=10, UseOwnerScale=true)
   &SkeletalMeshLocation(
    Bone Sampling Mode=::NewEnumerator4
    MaskSpawnGroup=false
    Orientation Sampling=::NewEnumerator1
    Sampled Position Offset=&AddVectorToPosition(Position=<-5 15 20>, Vector=$Particles.ShapeLocation.ShapeVector)
    Skeletal Mesh=$System.Skeletal Mesh
   )
   &AddVelocity(Module.Use Falloff Distance=false, Origin Offset=<0 10 20>, Override Seed=false, Velocity Mode=::NewEnumerator1, Velocity Speed=50) {enabled=false}
   &AddVelocity(Velocity=<0 50 0>) {enabled=false, module=AddVelocity001}
   &CameraOffset(Camera Offset Amount=-50)
  ]
  particleUpdateScript [
   &ParticleState()
   &ScaleMeshSize(Scale Factor=&VectorFromFloat(Value=&FloatFromCurve(FloatCurve=DataInterface() {propertyValues {Curve {keys [<at 0.7 1>,<at 1 0.6>]}}}, Scale Curve=0.7)))
   &InheritSourceMovement(
    Applied Position Falloff Curve=DataInterface() {propertyValues {Curve {keys [<at 0 1>,<at 1 0.4>]}}}
    Applied Position Falloff Distance=100
    Applied Position Scale=<0.7 0.7 1>
    ApplyForce=false
    ApplyPosition=true
    ApplyPositionFalloff=true
    ApplyPositionFalloffCurve=true
   )
   &UpdateMeshOrientation(Rotation Rate=&RandomRangeFloat(Minimum=1, Maximum=3), Rotation Vector=<1 0.2 0.4>)
   &CurlNoiseForce(Bias Noise=false, Noise Strength=800, Pan Noise=false) {enabled=false}
   &Drag(Drag=3, Use Linear Drag=true, Use Rotational Drag=true)
   &GravityForce(Gravity=<0 0 -100>)
   &SolveForcesAndVelocity(Clamp Velocity=false, Limit Acceleration=false)
   &ScaleColor(
    Scale Alpha=&FloatFromCurve(FloatCurve=DataInterface() {propertyValues {Curve {keys [<at 0.8 1>,<at 1 0>]}}})
    Scale RGB=&MakeVectorFromLinearColorRGB(Color=<rgba 0.7292 0.0489 0.0438 3>)
    ScaleA=true
    ScaleRGB=true
   )
   &DynamicMaterialParameters(
    Index 0 Param 1=&FloatFromCurve(
     FloatCurve=DataInterface() {propertyValues {Curve {keys [<at 0.3 1>,<at 1 -0.4>]}}}
     Scale Curve=&RandomRangeFloat(Minimum=0.8, Maximum=1)
    )
    Index 0 Param 2=&RandomRangeFloat(Minimum=0, Maximum=0.5)
    Index 0 Param 3=&RandomRangeFloat(Minimum=0, Maximum=0.5)
    Index 0 Param 4=&RandomRangeFloat(Minimum=0.2, Maximum=0.2)
    Module.Index0Float0Write=true
    Module.Index0Float1Write=true
    Module.Index0Float2Write=true
    Module.Index0Float3Write=true
   )
  ]
  rendererClasses [&NiagaraMeshRendererProperties]
  renderers [{rendererClass=&NiagaraMeshRendererProperties, rendererIndex=0}]
 }
 RoseFE {
  bEnabled=false
  emitterUpdateScript [
   &EmitterState(Inactive Response=::NewEnumerator1, Life Cycle Mode=::NewEnumerator1, Loop Delay=0.2, Loop Duration=1.2, UseLoopDelay=true)
   &SpawnBurst_Instantaneous(Use Spawn Probability=false, UseLoopCountLimit=false) {enabled=false}
   &SpawnRate(
    SpawnRate=&MakeCustomFloatFromBool(Bool=$User.bIsFrontEndPreview, True Float=2, False Float=0)
    Use Spawn Probability=false
   )
  ]
  particleSpawnScript [
   &InitializeParticle(
    Lifetime=1.5
    Mesh Scale Max=<1 1 1>
    Mesh Scale Min=<0.7 0.7 0.7>
    Mesh Scale Mode=::NewEnumerator2
    Sprite Size Mode=::NewEnumerator1
    UsePositionOffset=false
    Write Material Random=true
   )
   &InitialMeshOrientation(
    Mesh Orientation Mode=::NewEnumerator2
    Rotation=&RandomRangeVector(Minimum=<0 0 0>, Maximum=<1 1 1>)
    Use Rotation Vector=true
   )
   &ShapeLocation(MaskSpawnGroup=false, Offset Mode=::NewEnumerator0, Override Seed=false, Sphere Radius=10, UseOwnerScale=true)
   &SkeletalMeshLocation(
    Bone Sampling Mode=::NewEnumerator4
    MaskSpawnGroup=false
    Orientation Sampling=::NewEnumerator1
    Sampled Position Offset=&AddVectorToPosition(Position=<0 0 25>, Vector=$Particles.ShapeLocation.ShapeVector)
    Skeletal Mesh=$System.Skeletal Mesh
   )
   &AddVelocity(Module.Use Falloff Distance=false, Origin Offset=<0 0 20>, Override Seed=false, Velocity Mode=::NewEnumerator1, Velocity Speed=20)
   &AddVelocity(Velocity=<0 50 0>) {enabled=false, module=AddVelocity001}
  ]
  particleUpdateScript [
   &ParticleState()
   &ScaleMeshSize(Scale Factor=&VectorFromFloat(Value=&FloatFromCurve(FloatCurve=DataInterface() {propertyValues {Curve {keys [<at 0.7 1>,<at 1 0.6>]}}}, Scale Curve=0.7)))
   &InheritSourceMovement(
    Applied Position Falloff Curve=DataInterface() {propertyValues {Curve {keys [<at 0 1>,<at 1 0.4>]}}}
    Applied Position Falloff Distance=100
    Applied Position Scale=<0.7 0.7 1>
    ApplyForce=false
    ApplyPosition=true
    ApplyPositionFalloff=true
    ApplyPositionFalloffCurve=true
   ) {enabled=false}
   &UpdateMeshOrientation(Rotation Rate=&RandomRangeFloat(Minimum=0.2, Maximum=0.5), Rotation Vector=<1 0.2 0.4>)
   &CurlNoiseForce(Bias Noise=false, Noise Frequency=30, Noise Strength=50, Pan Noise=false) {enabled=false}
   &Drag(Drag=3, Use Linear Drag=true, Use Rotational Drag=true)
   &GravityForce(Gravity=<0 0 -60>)
   &SolveForcesAndVelocity(Clamp Velocity=false, Limit Acceleration=false)
   &ScaleColor(
    Scale Alpha=&FloatFromCurve(FloatCurve=DataInterface() {propertyValues {Curve {keys [<at 0.8 1>,<at 1 0>]}}})
    Scale RGB=&MakeVectorFromLinearColorRGB(Color=<rgba 0.6 0.0403 0.036 3>)
    ScaleA=true
    ScaleRGB=true
   )
   &DynamicMaterialParameters(
    Index 0 Param 1=&FloatFromCurve(
     FloatCurve=DataInterface() {propertyValues {Curve {keys [<at 0.3 1>,<at 1 -0.4>]}}}
     Scale Curve=&RandomRangeFloat(Minimum=0.8, Maximum=1)
    )
    Index 0 Param 2=&RandomRangeFloat(Minimum=0, Maximum=0.5)
    Index 0 Param 3=&RandomRangeFloat(Minimum=0, Maximum=0.5)
    Index 0 Param 4=&RandomRangeFloat(Minimum=0.2, Maximum=0.2)
    Module.Index0Float0Write=true
    Module.Index0Float1Write=true
    Module.Index0Float2Write=true
    Module.Index0Float3Write=true
   )
  ]
  rendererClasses [&NiagaraMeshRendererProperties]
  renderers [{rendererClass=&NiagaraMeshRendererProperties, rendererIndex=0}]
 }
}
summary {
 systemName=NS_Pickaxe_FlameBride_OffHand_Idle
 emitters [
  {emitterName=Fire, rendererClasses [&NiagaraSpriteRendererProperties]}
  {emitterName=Rose, rendererClasses [&NiagaraMeshRendererProperties]}
  {bEnabled=false, emitterName=FireFE, rendererClasses [&NiagaraSpriteRendererProperties]}
  {bEnabled=false, emitterName=RoseFE, rendererClasses [&NiagaraMeshRendererProperties]}
 ]
 userVariables [
  User.bIsFrontEnd {type=NiagaraBool}
  User.bIsFrontEndPreview {type=NiagaraBool}
  User.PhysicalSurface {type=NiagaraInt32}
  User.PlayerSkeletalMesh {
   type=NiagaraDataInterfaceSkeletalMesh
   defaultValue {
    struct=NiagaraExt_VariableValue_DataInterface
    value {dataInterface=&self:NiagaraDataInterfaceSkeletalMesh_0}
   }
  }
  User.SwingSpawn {type=NiagaraBool}
 ]
}
